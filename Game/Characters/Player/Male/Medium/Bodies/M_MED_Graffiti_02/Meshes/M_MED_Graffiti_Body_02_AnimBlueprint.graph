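# Post-process style AnimBlueprint for the Graffiti 02 male player body, copying the owner mesh pose and adding Control Rig, zipper and collar secondary motion.
# The zipper bone uses a weighted constraint plus AnimDynamics.
# Head-to-spine_05 yaw drives additive collar bone rotations, with separate alpha sets chosen by bIsTargeting.
# The head bone shrinks to near zero when the playlist uses custom character parts.
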AnimBlueprint /Game/Characters/Player/Male/Medium/Bodies/M_MED_Graffiti_02/Meshes/M_MED_Graffiti_Body_02_AnimBlueprint
target_skeleton=&/M_MED_Graffiti_Body_02_Skeleton
graphs [
 AnimGraph {nodes [
   n1 = "Transform (Modify) Bone - Bone: dyn_collar_front_l_01"(Alpha=0, Translation="0,0,0", Rotation="36,20,20", Scale="1,1,1") {
    rotationMode=BMM_Additive
    rotationSpace=BCS_BoneSpace
    alphaBoolBlend {blendOption=Linear}
    boneToModify {boneName=dyn_collar_front_l_01}
    rotation=<rot 36 20 20>
   }
   "Transform (Modify) Bone - Bone: dyn_collar_front_r_01"(ComponentPose=@n1, Alpha=0, Translation="0,0,0", Rotation="-10,-10,5", Scale="1,1,1") {
    rotationMode=BMM_Additive
    rotationSpace=BCS_BoneSpace
    alphaBoolBlend {blendOption=Linear}
    boneToModify {boneName=dyn_collar_front_r_01}
    rotation=<rot -10 -10 5>
   }
   "Transform (Modify) Bone - Bone: dyn_collar_front_l_01"(Alpha=0, Translation="0,0,0", Rotation="-5,20,-25", Scale="1,1,1") {
    rotationMode=BMM_Additive
    rotationSpace=BCS_BoneSpace
    alphaBoolBlend {blendOption=Linear}
    boneToModify {boneName=dyn_collar_front_l_01}
    rotation=<rot -5 20 -25>
   }
   n2 = "Transform (Modify) Bone - Bone: head"!(ComponentPose=Local To Component(LocalPose=Use cached pose 'Pre_HeadScale'()), Alpha=1) -> (Pose) {scaleMode=BMM_Replace, alphaBoolBlend {blendOption=Linear}, boneToModify {boneName=head}, scale=<0.01 0.01 0.01>}
   Output Pose(Result=Blend Poses by bool(
     BlendPose_0=Component To Local(ComponentPose=@n2)
     BlendPose_1=Use cached pose 'Pre_HeadScale'()
     BlendTime_0=0
     BlendTime_1=0
     bActiveValue=Get GetPlayerListUsesCustomCharacterParts() -> (GetPlayerListUsesCustomCharacterParts:bool)
    ) {blendType=Linear}) {layerGroup=None}
   n3 = Control Rig(Source=Copy Pose From Mesh(SourceMeshComponent=Get MeshToCopy() -> (MeshToCopy:Skeletal Mesh Component Object Reference)) {bUseAttachedParent=true}) {
    alphaBoolBlend {blendOption=Linear}
    controlRigAssetReference {blueprintRigClass=&/M_MED_Graffiti_Body_02_CtrlRig.M_MED_Graffiti_Body_02_CtrlRig_C}
   }
   n4 = "Constraint - dyn_zipper (PPT)"!(
    ComponentPose=Local To Component(LocalPose=@n3)
    Alpha=1
    ConstraintWeights_0=0.85
    ConstraintWeights_1:float=0.15
    ConstraintWeights_2:float=0
   ) -> (Pose) {
    alphaBoolBlend {blendOption=Linear}
    boneToModify {boneName=dyn_zipper}
    constraintSetup [
     {targetBone {boneName=pelvis}}
     {targetBone {boneName=thigh_twist_01_l}}
     {transformType=Translation, targetBone {boneName=thigh_twist_01_r}}
    ]
    constraintWeights [1,1,1]
   }
   n5 = Anim Dynamics - Bone: dyn_zipper(ComponentPose=@n4, Alpha=1, ExternalForce="0,0,-980") {
    angularDampingOverride=0.999
    angularSpringConstant=500
    bAngularSpring=true
    bOverrideAngularDamping=true
    bUseGravityOverride=true
    gravityScale=0
    simulationSpace=RootRelative
    alphaBoolBlend {blendOption=Linear}
    boundBone {boneName=dyn_zipper}
    componentAppliedLinearAccClamp=<3000 3000 3000>
    componentLinearAccScale=<0.45 0.4 0.45>
    componentLinearVelScale=<0.7 0.6 0.7>
    gravityOverride {z=-980}
    physicsBodyDefinitions [{
      boundBone {boneName=dyn_zipper}
      boxExtents=<10 10 10>
      constraintSetup {angularLimitsMax=<5 35 0>, angularLimitsMin=<-3 -30 -130>, angularTarget=<1 0 0>}
      localJointOffset=<1 0 0>
     }]
    relativeSpaceBone {boneName=root}
   }
   Save cached pose 'collarRotCache'(Pose=Component To Local(ComponentPose=@n5)) {cacheName=collarRotCache}
   n6 = Get ads_rotate_back() -> (ads_rotate_back:double)
   n7 = "Transform (Modify) Bone - Bone: dyn_collar_back_l_01"!(ComponentPose=Local To Component(LocalPose=Use cached pose 'collarRotCache'()), Alpha=@n6, Translation="0,0,0", Rotation="0,0,-40") -> (Pose) {
    rotationMode=BMM_Additive
    rotationSpace=BCS_BoneSpace
    scaleSpace=BCS_BoneSpace
    alphaBoolBlend {blendOption=Linear}
    boneToModify {boneName=dyn_collar_back_l_01}
    rotation {roll=-40}
    scale=<2 2 2>
   }
   n8 = "Transform (Modify) Bone - Bone: dyn_collar_back_r_01"!(ComponentPose=@n7, Alpha=@n6, Translation="0,0,0", Rotation="0,20,-40") -> (Pose) {
    rotationMode=BMM_Additive
    rotationSpace=BCS_BoneSpace
    scaleSpace=BCS_BoneSpace
    alphaBoolBlend {blendOption=Linear}
    boneToModify {boneName=dyn_collar_back_r_01}
    rotation {roll=-40, yaw=20}
    scale=<2 2 2>
   }
   n9 = "Transform (Modify) Bone - Bone: dyn_collar_front_r_01"(ComponentPose=@n8, Alpha=@n6, Translation="0,0,0", Rotation="150,-150,0", Scale="0,0,0") {
    rotationMode=BMM_Additive
    rotationSpace=BCS_BoneSpace
    alphaBoolBlend {blendOption=Linear}
    boneToModify {boneName=dyn_collar_front_r_01}
    rotation {pitch=150, yaw=-150}
    scale=<0 0 0>
   }
   n10 = Get ADS_Rot_Forward() -> (ADS_Rot_Forward:double)
   n11 = "Transform (Modify) Bone - Bone: dyn_collar_front_l_01"(ComponentPose=@n9, Alpha=@n10, Translation="0,0,0", Rotation="-5,20,-25", Scale="1,1,1") {
    rotationMode=BMM_Additive
    rotationSpace=BCS_BoneSpace
    alphaBoolBlend {blendOption=Linear}
    boneToModify {boneName=dyn_collar_front_l_01}
    rotation=<rot -5 20 -25>
   }
   n12 = "Transform (Modify) Bone - Bone: dyn_collar_front_r_01"(ComponentPose=@n11, Alpha=@n10, Translation="0,0,0", Rotation="-30,-35,40", Scale="1,1,1") {
    rotationMode=BMM_Additive
    rotationSpace=BCS_BoneSpace
    alphaBoolBlend {blendOption=Linear}
    boneToModify {boneName=dyn_collar_front_r_01}
    rotation=<rot -30 -35 40>
   }
   n13 = Get head_rot_back() -> (head_rot_back:double)
   n14 = "Transform (Modify) Bone - Bone: dyn_collar_back_l_01"(ComponentPose=Local To Component(LocalPose=Use cached pose 'collarRotCache'()), Alpha=@n13, Translation="0,0,0", Rotation="0,0,-40") {
    rotationMode=BMM_Additive
    rotationSpace=BCS_BoneSpace
    scaleSpace=BCS_BoneSpace
    alphaBoolBlend {blendOption=Linear}
    boneToModify {boneName=dyn_collar_back_l_01}
    rotation {roll=-40}
    scale=<2 2 2>
   }
   n15 = "Transform (Modify) Bone - Bone: dyn_collar_back_r_01"(ComponentPose=@n14, Alpha=@n13, Translation="0,0,0", Rotation="0,20,-40") {
    rotationMode=BMM_Additive
    rotationSpace=BCS_BoneSpace
    scaleSpace=BCS_BoneSpace
    alphaBoolBlend {blendOption=Linear}
    boneToModify {boneName=dyn_collar_back_r_01}
    rotation {roll=-40, yaw=20}
    scale=<2 2 2>
   }
   n16 = "Transform (Modify) Bone - Bone: dyn_collar_front_r_01"(ComponentPose=@n15, Alpha=@n13, Translation="0,0,0", Rotation="90,-90,0", Scale="0,0,0") {
    rotationMode=BMM_Additive
    rotationSpace=BCS_BoneSpace
    alphaBoolBlend {blendOption=Linear}
    boneToModify {boneName=dyn_collar_front_r_01}
    rotation {pitch=90, yaw=-90}
    scale=<0 0 0>
   }
   n17 = Get Head_Rot_Forward() -> (Head_Rot_Forward:double)
   n18 = "Transform (Modify) Bone - Bone: dyn_collar_front_l_01"(ComponentPose=@n16, Alpha=@n17, Translation="0,0,0", Rotation="-5,20,-25", Scale="1,1,1") {
    rotationMode=BMM_Additive
    rotationSpace=BCS_BoneSpace
    alphaBoolBlend {blendOption=Linear}
    boneToModify {boneName=dyn_collar_front_l_01}
    rotation=<rot -5 20 -25>
   }
   n19 = "Transform (Modify) Bone - Bone: dyn_collar_front_r_01"(ComponentPose=@n18, Alpha=@n17, Translation="0,0,0", Rotation="-30,-35,40", Scale="1,1,1") {
    rotationMode=BMM_Additive
    rotationSpace=BCS_BoneSpace
    alphaBoolBlend {blendOption=Linear}
    boneToModify {boneName=dyn_collar_front_r_01}
    rotation=<rot -30 -35 40>
   }
   Save cached pose 'Pre_HeadScale'(Pose=Blend Poses by bool(BlendPose_0=Component To Local(ComponentPose=@n12), BlendPose_1=Component To Local(ComponentPose=@n19), bActiveValue=Get bIsTargeting()) {blendType=Linear}) {cacheName=Pre_HeadScale}
  ]}
 EventGraph {nodes [
   Event BlueprintInitializeAnimation()
   > Set MeshToCopy(MeshToCopy:Skeletal Mesh Component Object Reference=GetOwnerSkeletalMeshComponent()) -> (Output_Get:Skeletal Mesh Component Object Reference)
   > Set GetPlayerListUsesCustomCharacterParts(GetPlayerListUsesCustomCharacterParts:bool=GetPlaylistUsesCustomCharacterParts!() -> (ReturnValue:bool)) -> (Output_Get:bool)
   Event BlueprintUpdateAnimation()
   > n1 = Sequence!(execute) -> (then_0, then_1, then_2:Exec, then_3:Exec, then_4:Exec)
   n2 = GetDeltaTransformFromRefPose(self=GetOwnerSkeletalMeshComponent(), BoneName=head, BaseName=spine_05)
   > Set headSpineRot(headSpineRot:double=BreakRotator(InRot=BreakTransform(InTransform=@n2).Rotation).Yaw) -> (Output_Get:double)
   Set head_rot_back(execute=@n1.then_1, head_rot_back:double=MapRangeClamped(Value=Get headSpineRot() -> (headSpineRot:double), InRangeA=0, InRangeB=-60)) -> (Output_Get:double)
   Set ads_rotate_back(execute=@n1.then_2, ads_rotate_back:double=MapRangeClamped(Value=Get headSpineRot(), InRangeA=0, InRangeB=-50, OutRangeA=0.5)) -> (Output_Get:double)
   Set Head_Rot_Forward(execute=@n1.then_3, Head_Rot_Forward:double=MapRangeClamped(Value=Get headSpineRot(), InRangeA=0, InRangeB=30)) -> (Output_Get:double)
   Set ADS_Rot_Forward(execute=@n1.then_4, ADS_Rot_Forward:double=MapRangeClamped(Value=Get headSpineRot(), InRangeA=0, InRangeB=30)) -> (Output_Get:double)
   TryGetPawnOwner()
   "To Float (Boolean)"(InBool=Get bIsTargeting())
  ]}
]
variables [
 ADS_Rot_Forward {type=number}
 ads_rotate_back {type=number}
 GetPlayerListUsesCustomCharacterParts {type=boolean}
 head_rot_back {type=number}
 Head_Rot_Forward {type=number}
 headSpineRot {type=number}
 MeshToCopy {type=SkeletalMeshComponent}
]
